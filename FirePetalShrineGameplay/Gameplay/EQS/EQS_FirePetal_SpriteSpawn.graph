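# EQS query that picks locations around the querier, apparently for spawning sprites in Fire Petal Shrine gameplay.
# A randomly jittered simple grid, with data-bound size and spacing, projects candidate points onto the navmesh using the default Sprites nav filter.
# Tests apply a data-bound minimum distance from the querier, filter on building and water containment, and check surface flatness and WorldStatic overlap.
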EnvQuery /FirePetalShrineGameplay/Gameplay/EQS/EQS_FirePetal_SpriteSpawn
options [{
  generator {
   name=FortEnvQueryGenerator_SimpleGrid_0
   class=FortEnvQueryGenerator_SimpleGrid
   properties {
    bAutoSortTests=true
    bCanRunAsync=false
    envQueryResultNormalizationOption=Default
    generateAround=&EnvQueryContext_Querier
    navDataOverrideContext=null
    optionName=""
    overridenAgentNameForNavmesh=None
    randomDisplacementRatio=0.35
    gridSize {
     dataBinding=&self:FortEnvQueryGenerator_SimpleGrid_0.FortAIDataProvider_ScalableFloat_1
     dataField=FloatValue
     defaultValue=100000
    }
    projectionData {
     bCanDisableTrace=true
     bCanProjectDown=true
     bCanTraceOnGeometry=true
     bCanTraceOnNavMesh=true
     bOnlyBlockingHits=true
     bTraceComplex=false
     extentX=0
     extentY=10
     extentZ=10
     navigationFilter=/Sprites/AI/NavFilters/NavFilter_DefaultSprites.NavFilter_DefaultSprites_C
     postProjectionVerticalOffset=0
     projectDown=10000
     projectUp=10000
     serializedChannel=ECC_GameTraceChannel11
     traceChannel=TraceTypeQuery10
     traceMode=Navigation
     traceProfileName=None
     traceShape=Line
     versionNum=1
    }
    spaceBetween {
     dataBinding=&self:FortEnvQueryGenerator_SimpleGrid_0.FortAIDataProvider_ScalableFloat_0
     dataField=FloatValue
     defaultValue=25000
    }
   }
  }
  tests [
   EnvQueryTest_Distance_1 {
    class=EnvQueryTest_Distance
    properties {
     distanceTo=&EnvQueryContext_Querier
     filterType=Minimum
     scoringEquation=Linear
     scoringFactor=1
     testComment=""
     testMode=Distance3D
     testPurpose=Filter
     floatValueMin {
      dataBinding=&self:EnvQueryTest_Distance_1.FortAIDataProvider_ScalableFloat_0
      dataField=FloatValue
      defaultValue=0
     }
    }
   }
   FortQueryTest_InsideBuilding_0 {class=FortQueryTest_InsideBuilding, properties {testPurpose=Filter}}
   FortQueryTest_InsideWater_0 {
    class=FortQueryTest_InsideWater
    properties {
     bDefineReferenceValue=false
     boolValue=false
     clampMaxType=None
     clampMinType=None
     filterType=Match
     floatValueMax=0
     floatValueMin=0
     multipleContextFilterOp=AllPass
     multipleContextScoreOp=AverageScore
     normalizationType=Absolute
     referenceValue=0
     scoreClampMax=0
     scoreClampMin=0
     scoringEquation=Constant
     scoringFactor=1
     testComment=""
     testPurpose=Filter
     testRadius=128
    }
   }
   FortQueryTest_OnFlatSurface_1 {class=FortQueryTest_OnFlatSurface, properties {radius=200, toleranceZ=120, traceOffsetUp=100}}
   EnvQueryTest_Overlap_1 {
    class=EnvQueryTest_Overlap
    properties {
     boolValue=true
     overlapData {bSkipOverlapQuerier=true, extentX=200, extentY=200, extentZ=200, overlapChannel=ECC_WorldStatic, shapeOffset=<0 0 0>}
    }
   }
  ]
 }]
