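# EQS query that returns shrine inject location actors in random order, likely for picking where a FirePetal shrine is injected.
# A single Fortnite ActorsOfClass generator gathers B_FirePetal_ShrineInjectLocation actors without restricting them to the search radius around the querier.
# Randomization defaults on through a scalable-float data binding.
# No tests filter or score the candidates, so the generator alone determines the result.
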EnvQuery /FirePetalShrineGameplay/Gameplay/EQS/EQS_RandomShrineInject
options [{
  generator {
   name=FortEnvQueryGenerator_ActorsOfClass_0
   class=FortEnvQueryGenerator_ActorsOfClass
   properties {
    bAutoSortTests=true
    bCanRunAsync=false
    envQueryResultNormalizationOption=Default
    generateOnlyActorsInRadius=false
    optionName=""
    searchCenter=&EnvQueryContext_Querier
    searchRadius=500
    searchedActorClass=/FirePetalShrineGameplay/Gameplay/Actors/B_FirePetal_ShrineInjectLocation.B_FirePetal_ShrineInjectLocation_C
    bRandomizeResults {
     dataBinding=&self:FortEnvQueryGenerator_ActorsOfClass_0.FortAIDataProvider_ScalableFloat_0
     dataField=BoolValue
     defaultValue=true
    }
   }
  }
  tests []
 }]
